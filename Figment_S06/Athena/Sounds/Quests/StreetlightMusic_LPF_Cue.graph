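# Sound cue for looping streetlight quest music, in a variant named LPF (likely low-pass filtered), assigned to the quest dance music sound class.
# A quality level node plays the standard LPF wave on the first two quality tiers and the Opt wave on the last two.
# The cue slightly boosts volume and overrides attenuation shape and falloff distance.
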SoundCue /Figment_S06/Athena/Sounds/Quests/StreetlightMusic_LPF_Cue
root=QualityLevel_0
nodes [
 {id=QualityLevel_0, children [WavePlayer_1,WavePlayer_1,WavePlayer_0,WavePlayer_0]}
 {id=WavePlayer_0, properties {bLooping=true, soundWaveAssetPtr=&/Streetlight_LPF_Music_Opt}}
 {id=WavePlayer_1, properties {bLooping=true, soundWaveAssetPtr=&/Streetlight_LPF_Music}}
]
properties {
 maxDistance=2097152
 soundClassObject=/Game/Sounds/FortSoundClassesAndModes/Fort_DanceMusic_Quest_PSM
 volumeMultiplier=1.2
 attenuationOverrides {falloffDistance=500, attenuationShapeExtents=<0 35>}
}
