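# Sound cue for the open-LPF variant of the streetlight music used by Figment S06 quests.
# A quality-level node feeds the first two levels from the main wave and the last two from its "_opt" variant.
# The cue has no filter node, plays pitched down, uses the quest dance-music sound class, and overrides attenuation for positional falloff.
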SoundCue /Figment_S06/Athena/Sounds/Quests/StreetlightMusic_LPF_Open_Cue
root=QualityLevel_0
nodes [
 {id=QualityLevel_0, children [WavePlayer_0,WavePlayer_0,WavePlayer_1,WavePlayer_1]}
 {id=WavePlayer_0, properties {soundWaveAssetPtr=&/StreetLight_LPF_Open}}
 {id=WavePlayer_1, properties {soundWaveAssetPtr=&/Streetlight_LPF_Open_opt}}
]
properties {
 bOverrideAttenuation=true
 maxDistance=5000
 pitchMultiplier=0.75
 soundClassObject=/Game/Sounds/FortSoundClassesAndModes/Fort_DanceMusic_Quest_PSM
 volumeMultiplier=1
 attenuationOverrides {falloffDistance=5000, attenuationShapeExtents=<0 35>}
}
